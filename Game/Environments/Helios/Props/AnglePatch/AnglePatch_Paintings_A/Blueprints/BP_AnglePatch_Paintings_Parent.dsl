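; Parent Blueprint for the AnglePatch painting props, exposing canvas adjustment and flipbook settings on a static mesh component.
; The construction script writes scale, position, rotation and flipbook rows, columns and speed into default custom primitive data slots, presumably read by the material.
; Rotation is converted from degrees to a fractional turn before it is written.
; BeginPlay, overlap and Tick events contain no active logic.
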(event EventBeginPlay
  (|CallParentFunction) ; This node is disabled and will not be called.\nDrag off pins to build functionality.
)

(event Collision|EventActorBeginOverlap
  (params (OtherActor :type Actor%20Object%20Reference)))

(event EventTick
  (params (DeltaSeconds :type Float%20%28single-precision%29)))

(fn ConstructionScript
  (bind _rotation (Variables|CanvasAdjustments|GetRotation))
  (bind _returnvalue (/ _rotation 3.6e+02))
  (bind _returnvalue_2 (Math|Float|Fraction _returnvalue))
  (|CallParentFunction)
  (Rendering|Material|SetDefaultCustomPrimitiveDataFloat (Variables|Debug|GetStaticMeshComponent) 0 (Variables|CanvasAdjustments|GetScaleGlobal))
  (Rendering|Material|SetDefaultCustomPrimitiveDataFloat (Variables|Debug|GetStaticMeshComponent) 1 (Variables|CanvasAdjustments|GetScaleHorizontal))
  (Rendering|Material|SetDefaultCustomPrimitiveDataFloat (Variables|Debug|GetStaticMeshComponent) 2 (Variables|CanvasAdjustments|GetScaleVertical))
  (Rendering|Material|SetDefaultCustomPrimitiveDataFloat (Variables|Debug|GetStaticMeshComponent) 3 (Variables|CanvasAdjustments|GetPositionHorizontal))
  (Rendering|Material|SetDefaultCustomPrimitiveDataFloat (Variables|Debug|GetStaticMeshComponent) 4 (Variables|CanvasAdjustments|GetPositionVertical))
  (Rendering|Material|SetDefaultCustomPrimitiveDataFloat (Variables|Debug|GetStaticMeshComponent) 4 (Variables|CanvasAdjustments|GetPositionVertical))
  (Rendering|Material|SetDefaultCustomPrimitiveDataFloat (Variables|Debug|GetStaticMeshComponent) 5 _returnvalue_2)
  (Rendering|Material|SetDefaultCustomPrimitiveDataFloat (Variables|Debug|GetStaticMeshComponent) 5 _returnvalue_2)
  (Rendering|Material|SetDefaultCustomPrimitiveDataFloat (Variables|Debug|GetStaticMeshComponent) 6 (Variables|CanvasAdjustments|GetFlipbookNumRows))
  (Rendering|Material|SetDefaultCustomPrimitiveDataFloat (Variables|Debug|GetStaticMeshComponent) 12 (Variables|CanvasAdjustments|GetFlipbookNumColumns))
  (Rendering|Material|SetDefaultCustomPrimitiveDataFloat (Variables|Debug|GetStaticMeshComponent) 16 (Variables|CanvasAdjustments|GetFlipbookAnimSpeed)))
